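# Translucent deferred decal material for tire tread marks, filed with the Desert environment's tire track decals.
# The alpha of the T_OEM_TreadMarks texture is inverted through a OneMinus node, forming a tread mask.
# The only other nodes are a constant color left at default and two scalar constants, with no parameters exposed.
# The graph does not show which material outputs these nodes drive.
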Material /FigmentCore/Environments/Desert/Materials/TireTracks_Decals/M_TireMarks_Decal
graph {
 properties {blendMode=BLEND_Translucent, materialDomain=MD_DeferredDecal, pixelDepthOffsetMode=PDOM_Legacy, usedShadingModels=Default Lit}
 nodes [
  Constant3Vector()
  Constant()
  Constant() {r=0.7}
  OneMinus(None=TextureSample() {texture=&/T_OEM_TreadMarks}.A)
 ]
}
