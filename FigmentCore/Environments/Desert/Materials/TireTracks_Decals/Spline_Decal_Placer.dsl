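; Construction-script actor that lays decals along an editable spline, placed among the desert tire track decal assets.
; It first snaps each spline point to the surface with a vertical line trace within a threshold.
; It then adds decal components at intervals set by X scale and clamped offset, with a capped count derived from spline length.
; Each decal projects downward, follows the spline's yaw and roll, and takes the configured material.
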(event EventBeginPlay)

(event Collision|EventActorBeginOverlap
  (params (OtherActor :type Actor%20Object%20Reference)))

(event EventTick
  (params (DeltaSeconds :type Float%20%28single-precision%29)))

(fn ConstructionScript
  (Utilities|FlowControl|Sequence
    (:then_0
      ;[1 1 1] Snap Spline Point to Surface
      (for _index (range (Spline|GetNumberOfSplinePoints (Variables|Default|GetSpline_Decal_Placer)))
        (bind (_returnvalue_x _returnvalue_y _returnvalue_z) (Math|Vector|BreakVector (Spline|GetLocationatSplinePoint (Variables|Default|GetSpline_Decal_Placer) _index "World")))
        (bind (_returnvalue_x_2 _returnvalue_y_2 _returnvalue_z_2) (Math|Vector|BreakVector (Spline|GetLocationatSplinePoint (Variables|Default|GetSpline_Decal_Placer) _index "World")))
        (bind (_outhit _returnvalue) (Collision|LineTraceByChannel (Math|Vector|MakeVector _returnvalue_x_2 _returnvalue_y_2 (+ _returnvalue_z_2 (Variables|Default|GetDecalSnapThreshhold))) (Math|Vector|MakeVector _returnvalue_x _returnvalue_y (- _returnvalue_z (Variables|Default|GetDecalSnapThreshhold)))))
        (bind (_outhit_bblockinghit _outhit_binitialoverlap _outhit_time _outhit_distance _outhit_location) (Collision|BreakHitResult _outhit))
        (if _returnvalue
          (Spline|SetLocationatSplinePoint (Variables|Default|GetSpline_Decal_Placer) _index _outhit_location "World" true))))
    (:then_1
      (for _index_2 (range (+ (Math|Float|Truncate (Math|Float|Clamp(Float) (Math|Fortnite|%28double%29integer64/%28double%29integer64 (Math|Conversions|ToInteger64(Float) (Spline|GetSplineLength (Variables|Default|GetSpline_Decal_Placer))) (Math|Conversions|ToInteger64(Float) (* (* (Variables|Setup|GetDecalScaleX) 512.0) (Math|Float|Clamp(Float) (Variables|Setup|GetDecalOffset) 0.0 12.0)))) 0.0 1e+02)) 1))
        (bind _returnvalue_2 (AddComponent|NODE_AddDecalComponent-0 :RelativeTransform_Location "0, 0, 0" :RelativeTransform_Rotation "90.000000,0.000000,0.000000" :RelativeTransform_Scale_X (Variables|Setup|GetProjectionScale) :RelativeTransform_Scale_Y (Variables|Setup|GetDecalScaleY) :RelativeTransform_Scale_Z (Variables|Setup|GetDecalScaleX)))
        (bind (_returnvalue_roll _returnvalue_pitch _returnvalue_yaw) (Math|Rotator|BreakRotator (Spline|GetRotationatDistanceAlongSpline (Variables|Default|GetSpline_Decal_Placer) (Math|Conversions|ToFloat(Integer64) (* _index_2 (* (* (Variables|Setup|GetDecalScaleX) 512.0) (Math|Float|Clamp(Float) (Variables|Setup|GetDecalOffset) 0.0 12.0)))) "World")))
        (Transformation|SetWorldLocationAndRotation _returnvalue_2 (Spline|GetLocationatDistanceAlongSpline (Variables|Default|GetSpline_Decal_Placer) (Math|Conversions|ToFloat(Integer64) (* _index_2 (* (* (Variables|Setup|GetDecalScaleX) 512.0) (Math|Float|Clamp(Float) (Variables|Setup|GetDecalOffset) 0.0 12.0)))) "World") (Math|Rotator|MakeRotator _returnvalue_roll 9e+01 _returnvalue_yaw) false false)
        (Rendering|Components|Decal|SetDecalMaterial _returnvalue_2 (Variables|Setup|GetDecalMaterial))))))
